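# Unlit material in the Outlander Shock Tower gadget folder, built around particle color for an emissive effect.
# Particle color RGB feeds the Emissive input of the shared MF_FX_EmissiveAdjustments function, with no link from its output to a material output shown.
# An unconnected vector parameter holds a bright HDR blue.
# Despite the name, the graph samples no texture.
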Material /Game/CharClasses/Outlander/Gadgets/Fragment/ShockTower/EmissiveTexturedMaterial
graph {
 properties {pixelDepthOffsetMode=PDOM_Legacy, shadingModel=MSM_Unlit, usedShadingModels=Unlit}
 nodes [
  MaterialFunctionCall(Emissive=ParticleColor().RGB) {materialFunction=&MF_FX_EmissiveAdjustments}
  VectorParameter() {defaultValue=<rgba 0.186 0.3194 10 0>}
 ]
}
