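# Unlit additive particle-sprite material for the lightning zap effect of the Outlander Shock Tower gadget.
# The zap texture's RGB is tinted by particle color and fed into the shared MF_FX_EmissiveAdjustments function.
# A separate chain multiplies texture alpha by boosted particle alpha, then inverts and clamps it, with no destination shown for the result.
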Material /Game/CharClasses/Outlander/Gadgets/Fragment/ShockTower/Materials/M_Outlander_ShockTower_LightningZap
graph {
 properties {bUsedWithParticleSprites=true, blendMode=BLEND_Additive, pixelDepthOffsetMode=PDOM_Legacy, shadingModel=MSM_Unlit, usedShadingModels=Unlit}
 nodes [
  n1 = TextureSample() {texture=/Game/CharClasses/Outlander/Gadgets/Fragment/ShockTower/Textures/T_ShockTower_LightningZap}
  n2 = ParticleColor()
  n3 = Multiply(A=@n2.A) {constB=5}
  Clamp(None=OneMinus(None=Multiply(A=@n1.A, B=@n3)))
  MaterialFunctionCall(Emissive=Multiply(A=@n1.RGB, B=@n2.RGB)) {materialFunction=&MF_FX_EmissiveAdjustments}
 ]
}
