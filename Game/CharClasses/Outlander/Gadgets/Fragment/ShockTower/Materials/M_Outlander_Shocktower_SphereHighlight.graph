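# Translucent unlit mesh-particle material, named as the sphere highlight for the Outlander Shock Tower gadget effect.
# Particle color feeds the emissive adjustments function, while particle alpha scales an inverted Fresnel term softened by depth fade.
# A caustics texture, masked by a parabolic band across U, is added to that alpha term.
# Its power exponent rises with cubed particle relative time, thinning the caustics late in the particle's life.
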Material /Game/CharClasses/Outlander/Gadgets/Fragment/ShockTower/Materials/M_Outlander_Shocktower_SphereHighlight
graph {
 properties {bUsedWithMeshParticles=true, blendMode=BLEND_Translucent, pixelDepthOffsetMode=PDOM_Legacy, shadingModel=MSM_Unlit, usedShadingModels=Unlit}
 nodes [
  n1 = ParticleColor()
  n2 = Add(A=MaterialFunctionCall(Exponent=Constant() {r=2}) {materialFunction=&MF_Fresnel_FX}.Fresnel Inverted) {constB=0.3}
  n3 = DepthFade(FadeDistance=Constant() {r=6})
  n4 = ComponentMask(None=TextureCoordinate()) {r=true}
  n5 = Multiply(A=@n4, B=OneMinus(None=@n4))
  n6 = Multiply(A=TextureSample(UVs=TextureCoordinate()) {texture=&T_ShieldCaustics}.RGB) {constB=0.98}
  n7 = LinearInterpolate(Alpha=Power(Base=ParticleRelativeTime()) {constExponent=3}) {constA=1, constB=16}
  Add(A=Multiply(A=@n1.A, B=Multiply(A=@n2, B=@n3)), B=Multiply(A=Multiply(A=@n5) {constB=4}, B=Power(Base=@n6, Exp=@n7)))
  MaterialFunctionCall(Emissive=@n1.RGB) {materialFunction=&MF_FX_EmissiveAdjustments}
 ]
}
